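# Animation Blueprint for the female medium Shapeshifter head on the Fortnite base head skeleton, with only event logic shown.
# On initialize it casts the pawn owner to FortPlayerPawn and caches it, while update just calls the parent.
# A chain that stores the Face part mesh as SnakeHairMesh and hides it has no incoming execution link.
# The OnCharacterCustomizationLoaded event is empty.
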AnimBlueprint /Game/Characters/Player/Female/Medium/Heads/F_MED_Shapeshifter/Meshes/Female_Medium_Shapeshifter_Head_AnimBP
target_skeleton=/Game/Characters/Player/Common/Fortnite_Base_Head/Fortnite_Base_Head_Export_Skeleton
graphs [EventGraph {nodes [
   n1 = Event BlueprintUpdateAnimation()
   > Parent: BlueprintUpdateAnimation!(execute, DeltaTimeX:float=@n1.DeltaTimeX) -> (then) {is=title K2Node_CallParentFunction}
   Event BlueprintInitializeAnimation()
   > n2 = Cast To FortPlayerPawn(Object=TryGetPawnOwner())
   > Set FortPlayerPawn(FortPlayerPawn:Fort Player Pawn Object Reference=@n2) -> (Output_Get:Fort Player Pawn Object Reference)
   n3 = Set SnakeHairMesh(SnakeHairMesh:Skeletal Mesh Component Object Reference=GetSkeletalMeshForPartType(self=Get FortPlayerPawn() -> (FortPlayerPawn:Fort Player Pawn Object Reference), PartType=Face)) -> (Output_Get:Skeletal Mesh Component Object Reference)
   > SetVisibility!(execute, self:Scene Component Object Reference=@n3, bNewVisibility:bool=false, bPropagateToChildren:bool=false) -> (then)
   Event OnCharacterCustomizationLoaded()
  ]}]
variables [
 FortPlayerPawn {type=FortPlayerPawn}
 SnakeHairMesh {type=SkeletalMeshComponent}
]
